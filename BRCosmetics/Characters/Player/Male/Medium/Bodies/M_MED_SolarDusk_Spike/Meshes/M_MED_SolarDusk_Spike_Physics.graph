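# Physics asset for the M_MED_SolarDusk_Spike male medium body cosmetic, covering the upper spine, pectorals and a two-bone necklace with capsule bodies.
# Angularly locked constraints tie the spine and pectoral bodies into a rigid torso shape, likely as collision for the necklace.
# The necklace chains from spine_05 through necklace_01 to necklace_02 on unlocked constraints with narrow swing and twist angles, suggesting simulated dangle motion.
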PhysicsAsset /BRCosmetics/Characters/Player/Male/Medium/Bodies/M_MED_SolarDusk_Spike/Meshes/M_MED_SolarDusk_Spike_Physics
bodies {
 deform_pec_aim_l [{shapeType=Capsule, length=4.7783, radius=4.9093, center=<5.377 -3.5375 0>, rotation=<rot 90 17.1912 0>}]
 deform_pec_aim_r [{shapeType=Capsule, length=4.7783, radius=4.9093, center=<-5.377 3.5375 0>, rotation=<rot 90 -162.8088 0>}]
 necklace_01 [{shapeType=Capsule, length=6.0834, radius=4.7703, shapeName=necklace_01_capsule, center=<4 0 4>, rotation=<rot 0 0 -88.6334>}]
 necklace_02 [{shapeType=Capsule, length=4.7471, radius=4.3788, shapeName=necklace_02_capsule, center=<5 0 2.3>, rotation=<rot 0 0 90>}]
 spine_03 [{shapeType=Capsule, length=1, radius=1}]
 spine_04 [{shapeType=Capsule, length=8.3327, radius=13.1778}]
 spine_05 [{shapeType=Capsule, length=11.275, radius=9.9741}]
}
constraints [
 {
  bone1Name=spine_05
  bone2Name=spine_04
  swing1LimitDegrees=45
  swing1Motion=Locked
  swing2LimitDegrees=45
  swing2Motion=Locked
  twistLimitDegrees=45
  twistMotion=Locked
 }
 {
  bone1Name=spine_04
  bone2Name=spine_03
  swing1LimitDegrees=45
  swing1Motion=Locked
  swing2LimitDegrees=45
  swing2Motion=Locked
  twistLimitDegrees=45
  twistMotion=Locked
 }
 {
  bone1Name=deform_pec_aim_l
  bone2Name=spine_04
  swing1LimitDegrees=45
  swing1Motion=Locked
  swing2LimitDegrees=45
  swing2Motion=Locked
  twistLimitDegrees=45
  twistMotion=Locked
 }
 {
  bone1Name=deform_pec_aim_r
  bone2Name=spine_04
  swing1LimitDegrees=45
  swing1Motion=Locked
  swing2LimitDegrees=45
  swing2Motion=Locked
  twistLimitDegrees=45
  twistMotion=Locked
 }
 {bone1Name=necklace_01, bone2Name=spine_05, swing1LimitDegrees=12, swing2LimitDegrees=20, twistLimitDegrees=1}
 {bone1Name=necklace_02, bone2Name=necklace_01, swing1LimitDegrees=12, swing2LimitDegrees=20, twistLimitDegrees=1}
]
