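# Animation Blueprint for the SolarDusk Spike male body cosmetic, adding secondary motion to a pose copied from the attached parent mesh.
# A belt corrective constraint and movement-driven Control Rig feed AnimDynamics for belt, badge and fur.
# RigidBody simulation follows, with bIsInFrontEnd switching between front-end and default physics assets.
# AnimDynamics then moves bracelets, shoelaces and shorts, with shoelace constraints midway; dyn_shorts_r_fr is simulated twice.
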AnimBlueprint /BRCosmetics/Characters/Player/Male/Medium/Bodies/M_MED_SolarDusk_Spike/Meshes/M_MED_SolarDusk_Spike_AnimBP
target_skeleton=&/M_MED_SolarDusk_Spike_Skeleton
graphs [
 AnimGraph {nodes [
   n1 = RigidBody!(ComponentPose=Local To Component(LocalPose=Use cached pose 'fe'()), Alpha, OverrideWorldGravity:Vector="0,0,-980", ExternalForce) -> (Pose) {
    bClampLinearTranslationLimitToRefPose=true
    bForceDisableCollisionBetweenConstraintBodies=true
    overridePhysicsAsset=&/M_MED_SolarDusk_Spike_Physics_fe
    simulationSpace=BaseBoneSpace
    alphaBoolBlend {blendOption=Linear}
    baseBoneRef {boneName=spine_05}
    simSpaceSettings {
     dampingAlpha=0
     maxAngularAcceleration=80
     maxAngularVelocity=8
     maxLinearAcceleration=700
     maxLinearVelocity=30
     velocityScaleZ=0.7
     worldAlpha=1
     externalLinearDragV {x=0.5}
    }
   }
   n2 = Component To Local(ComponentPose=@n1)
   n3 = RigidBody(ComponentPose=Local To Component(LocalPose=Use cached pose 'fe'()), OverrideWorldGravity="0,0,-980") {
    bClampLinearTranslationLimitToRefPose=true
    bForceDisableCollisionBetweenConstraintBodies=true
    overridePhysicsAsset=&/M_MED_SolarDusk_Spike_Physics
    simulationSpace=BaseBoneSpace
    alphaBoolBlend {blendOption=Linear}
    baseBoneRef {boneName=spine_05}
    simSpaceSettings {
     dampingAlpha=0
     maxAngularAcceleration=80
     maxAngularVelocity=8
     maxLinearAcceleration=700
     maxLinearVelocity=30
     velocityScaleZ=0.7
     worldAlpha=1
     externalLinearDragV {x=0.5}
    }
   }
   n4 = Component To Local(ComponentPose=@n3)
   n5 = Get bIsInFrontEnd()
   n6 = Anim Dynamics - Bone: bracelet_03!(
    ComponentPose=Local To Component(LocalPose=Blend Poses by bool(BlendPose_0=@n2, BlendPose_1=@n4, bActiveValue=@n5))
    Alpha=1
    GravityOverride:Vector="0,0,-980"
    ExternalForce="0,0,0"
   ) -> (Pose) {
    angularDampingOverride=1
    angularSpringConstant=700
    bAngularSpring=true
    bLinearSpring=true
    bOverrideAngularDamping=true
    bOverrideLinearDamping=true
    bUseGravityOverride=true
    gravityScale=0.35
    linearDampingOverride=1
    linearSpringConstant=10
    simulationSpace=RootRelative
    alphaBoolBlend {blendOption=Linear}
    boundBone {boneName=bracelet_03}
    componentLinearAccScale=<1 1 1>
    componentLinearVelScale=<1 1 1>
    gravityOverride {z=-980}
    physicsBodyDefinitions [{
      boundBone {boneName=bracelet_03}
      boxExtents=<10 10 10>
      constraintSetup {angularLimitsMax=<3 3 3>, angularLimitsMin=<-3 -3 -3>, angularTarget=<1 0 0>}
      localJointOffset=<0 0 0.1>
     }]
    relativeSpaceBone {boneName=thigh_r}
   }
   n7 = Anim Dynamics - Bone: bracelet_04!(ComponentPose=@n6, Alpha=1, GravityOverride:Vector="0,0,-980", ExternalForce="0,0,0") -> (Pose) {
    angularDampingOverride=1
    angularSpringConstant=700
    bAngularSpring=true
    bLinearSpring=true
    bOverrideAngularDamping=true
    bOverrideLinearDamping=true
    bUseGravityOverride=true
    gravityScale=0.35
    linearDampingOverride=1
    linearSpringConstant=10
    simulationSpace=RootRelative
    alphaBoolBlend {blendOption=Linear}
    boundBone {boneName=bracelet_04}
    componentLinearAccScale=<1 1 1>
    componentLinearVelScale=<1 1 1>
    gravityOverride {z=-980}
    physicsBodyDefinitions [{
      boundBone {boneName=bracelet_04}
      boxExtents=<10 10 10>
      constraintSetup {angularLimitsMax=<3 3 3>, angularLimitsMin=<-3 -3 -3>, angularTarget=<1 0 0>}
      localJointOffset=<0 0 0.1>
     }]
    relativeSpaceBone {boneName=thigh_r}
   }
   n8 = Anim Dynamics - Bone: shoelaces_r!(ComponentPose=@n7, Alpha=1, GravityOverride:Vector="0,0,-980", ExternalForce="0,0,0") -> (Pose) {
    angularDampingOverride=1
    angularSpringConstant=500
    bAngularSpring=true
    bOverrideAngularDamping=true
    bUseGravityOverride=true
    gravityScale=0.35
    linearSpringConstant=2000
    simulationSpace=RootRelative
    alphaBoolBlend {blendOption=Linear}
    boundBone {boneName=shoelaces_r}
    componentLinearAccScale=<1 1 1>
    componentLinearVelScale=<1 1 1>
    gravityOverride {z=-980}
    physicsBodyDefinitions [{
      boundBone {boneName=shoelaces_r}
      boxExtents=<10 10 10>
      constraintSetup {angularLimitsMax=<1 -2 10>, angularLimitsMin=<-1 -10 -10>, angularTarget=<1 0 0>}
      localJointOffset=<0.4 0 0>
     }]
    relativeSpaceBone {boneName=thigh_r}
   }
   n9 = Anim Dynamics - Bone: shoelaces_l!(ComponentPose=@n8, Alpha=1, GravityOverride:Vector="0,0,-980", ExternalForce="0,0,0") -> (Pose) {
    angularDampingOverride=1
    angularSpringConstant=500
    bAngularSpring=true
    bOverrideAngularDamping=true
    bUseGravityOverride=true
    gravityScale=0.35
    linearSpringConstant=2000
    simulationSpace=RootRelative
    alphaBoolBlend {blendOption=Linear}
    boundBone {boneName=shoelaces_l}
    componentLinearAccScale=<1 1 1>
    componentLinearVelScale=<1 1 1>
    gravityOverride {z=-980}
    physicsBodyDefinitions [{
      boundBone {boneName=shoelaces_l}
      boxExtents=<10 10 10>
      constraintSetup {angularLimitsMax=<1 -2 10>, angularLimitsMin=<-1 -10 -10>, angularTarget=<1 0 0>}
      localJointOffset=<0.4 0 0>
     }]
    relativeSpaceBone {boneName=thigh_r}
   }
   n10 = "Constraint - shoelaces_r (PPPPPP)"!(
    ComponentPose=@n9
    Alpha=1
    ConstraintWeights_0=0.5453
    ConstraintWeights_1:float=0.0005
    ConstraintWeights_2:float=0.0802
    ConstraintWeights_3:float=0.0595
    ConstraintWeights_4:float=0.1431
    ConstraintWeights_5:float=0.1714
   ) -> (Pose) {
    boneToModify {boneName=shoelaces_r}
    constraintSetup [
     {targetBone {boneName=pelvis}}
     {targetBone {boneName=spine_02}}
     {targetBone {boneName=thigh_twist_01_l}}
     {targetBone {boneName=deform_groin_l}}
     {targetBone {boneName=thigh_twist_01_r}}
     {targetBone {boneName=deform_groin_r}}
    ]
    constraintWeights [1,1,1,1,1,1]
   }
   n11 = "Constraint - shoelaces_l (PPPPPP)"!(
    ComponentPose=@n10
    Alpha=1
    ConstraintWeights_0=0.5342
    ConstraintWeights_1:float=0.0005
    ConstraintWeights_2:float=0.1481
    ConstraintWeights_3:float=0.176
    ConstraintWeights_4:float=0.0818
    ConstraintWeights_5:float=0.0595
   ) -> (Pose) {
    boneToModify {boneName=shoelaces_l}
    constraintSetup [
     {targetBone {boneName=pelvis}}
     {targetBone {boneName=spine_02}}
     {targetBone {boneName=thigh_twist_01_l}}
     {targetBone {boneName=deform_groin_l}}
     {targetBone {boneName=thigh_twist_01_r}}
     {targetBone {boneName=deform_groin_r}}
    ]
    constraintWeights [1,1,1,1,1,1]
   }
   n12 = Anim Dynamics - Bone: dyn_shorts_l_bk!(ComponentPose=@n11, Alpha=1, GravityOverride:Vector="0,0,-980", ExternalForce="0,0,0") -> (Pose) {
    angularDampingOverride=1
    angularSpringConstant=500
    bAngularSpring=true
    bOverrideAngularDamping=true
    bUseGravityOverride=true
    gravityScale=0.35
    linearSpringConstant=2000
    simulationSpace=RootRelative
    alphaBoolBlend {blendOption=Linear}
    boundBone {boneName=dyn_shorts_l_bk}
    componentLinearAccScale=<1 1 1>
    componentLinearVelScale=<1 1 1>
    gravityOverride {z=-980}
    physicsBodyDefinitions [{
      boundBone {boneName=dyn_shorts_l_bk}
      boxExtents=<10 10 10>
      constraintSetup {angularLimitsMax=<1 0 0>, angularLimitsMin=<-1 -20 0>, angularTarget=<1 0 0>}
      localJointOffset=<0.4 0 0>
     }]
    relativeSpaceBone {boneName=thigh_r}
   }
   n13 = Anim Dynamics - Bone: dyn_shorts_l_fr!(ComponentPose=@n12, Alpha=1, GravityOverride:Vector="0,0,-980", ExternalForce="0,0,0") -> (Pose) {
    angularDampingOverride=1
    angularSpringConstant=500
    bAngularSpring=true
    bOverrideAngularDamping=true
    bUseGravityOverride=true
    gravityScale=0.35
    linearSpringConstant=2000
    simulationSpace=RootRelative
    alphaBoolBlend {blendOption=Linear}
    boundBone {boneName=dyn_shorts_l_fr}
    componentAppliedLinearAccClamp=<50000 50000 50000>
    componentLinearAccScale=<1 1 1>
    componentLinearVelScale=<1 1 1>
    gravityOverride {z=-980}
    physicsBodyDefinitions [{
      boundBone {boneName=dyn_shorts_l_fr}
      boxExtents=<10 10 10>
      constraintSetup {angularLimitsMax=<0 4 0>, angularLimitsMin=<0 -20 0>, angularTarget=<1 0 0>}
      localJointOffset=<0.4 0 0>
     }]
    relativeSpaceBone {boneName=thigh_r}
   }
   n14 = Anim Dynamics - Bone: dyn_shorts_l_iner!(ComponentPose=@n13, Alpha=1, GravityOverride:Vector="0,0,-980", ExternalForce="0,0,0") -> (Pose) {
    angularDampingOverride=1
    angularSpringConstant=500
    bAngularSpring=true
    bOverrideAngularDamping=true
    bUseGravityOverride=true
    gravityScale=0.35
    linearSpringConstant=2000
    simulationSpace=RootRelative
    alphaBoolBlend {blendOption=Linear}
    boundBone {boneName=dyn_shorts_l_iner}
    componentLinearAccScale=<1 1 1>
    componentLinearVelScale=<1 1 1>
    gravityOverride {z=-980}
    physicsBodyDefinitions [{
      boundBone {boneName=dyn_shorts_l_iner}
      boxExtents=<10 10 10>
      constraintSetup {angularLimitsMax=<1 0 0>, angularLimitsMin=<-1 -20 0>, angularTarget=<1 0 0>}
      localJointOffset=<0.4 0 0>
     }]
    relativeSpaceBone {boneName=thigh_r}
   }
   n15 = Anim Dynamics - Bone: dyn_shorts_l_side!(ComponentPose=@n14, Alpha=1, GravityOverride:Vector="0,0,-980", ExternalForce="0,0,0") -> (Pose) {
    angularDampingOverride=1
    angularSpringConstant=500
    bAngularSpring=true
    bOverrideAngularDamping=true
    bUseGravityOverride=true
    gravityScale=0.35
    linearSpringConstant=2000
    simulationSpace=RootRelative
    alphaBoolBlend {blendOption=Linear}
    boundBone {boneName=dyn_shorts_l_side}
    componentAppliedLinearAccClamp=<50000 50000 50000>
    componentLinearAccScale=<1 1 1>
    componentLinearVelScale=<1 1 1>
    gravityOverride {z=-980}
    physicsBodyDefinitions [{
      boundBone {boneName=dyn_shorts_l_side}
      boxExtents=<10 10 10>
      constraintSetup {angularLimitsMax=<0 4 0>, angularLimitsMin=<0 -20 0>, angularTarget=<1 0 0>}
      localJointOffset=<0.4 0 0>
     }]
    relativeSpaceBone {boneName=thigh_r}
   }
   n16 = Anim Dynamics - Bone: dyn_shorts_r_iner!(ComponentPose=@n15, Alpha=1, GravityOverride:Vector="0,0,-980", ExternalForce="0,0,0") -> (Pose) {
    angularDampingOverride=1
    angularSpringConstant=500
    bAngularSpring=true
    bOverrideAngularDamping=true
    bUseGravityOverride=true
    gravityScale=0.35
    linearSpringConstant=2000
    simulationSpace=RootRelative
    alphaBoolBlend {blendOption=Linear}
    boundBone {boneName=dyn_shorts_r_iner}
    componentLinearAccScale=<1 1 1>
    componentLinearVelScale=<1 1 1>
    gravityOverride {z=-980}
    physicsBodyDefinitions [{
      boundBone {boneName=dyn_shorts_r_iner}
      boxExtents=<10 10 10>
      constraintSetup {angularLimitsMax=<1 0 0>, angularLimitsMin=<-1 -20 0>, angularTarget=<1 0 0>}
      localJointOffset=<0.4 0 0>
     }]
    relativeSpaceBone {boneName=thigh_r}
   }
   n17 = Anim Dynamics - Bone: dyn_shorts_r_fr!(ComponentPose=@n16, Alpha=1, GravityOverride:Vector="0,0,-980", ExternalForce="0,0,0") -> (Pose) {
    angularDampingOverride=1
    angularSpringConstant=500
    bAngularSpring=true
    bOverrideAngularDamping=true
    bUseGravityOverride=true
    gravityScale=0.35
    linearSpringConstant=2000
    simulationSpace=RootRelative
    alphaBoolBlend {blendOption=Linear}
    boundBone {boneName=dyn_shorts_r_fr}
    componentAppliedLinearAccClamp=<50000 50000 50000>
    componentLinearAccScale=<1 1 1>
    componentLinearVelScale=<1 1 1>
    gravityOverride {z=-980}
    physicsBodyDefinitions [{
      boundBone {boneName=dyn_shorts_r_fr}
      boxExtents=<10 10 10>
      constraintSetup {angularLimitsMax=<0 4 0>, angularLimitsMin=<0 -20 0>, angularTarget=<1 0 0>}
      localJointOffset=<0.4 0 0>
     }]
    relativeSpaceBone {boneName=thigh_r}
   }
   n18 = Anim Dynamics - Bone: dyn_shorts_r_fr(ComponentPose=@n17, Alpha=1, GravityOverride="0,0,-980", ExternalForce="0,0,0") {
    angularDampingOverride=1
    angularSpringConstant=500
    bAngularSpring=true
    bOverrideAngularDamping=true
    bUseGravityOverride=true
    gravityScale=0.35
    linearSpringConstant=2000
    simulationSpace=RootRelative
    alphaBoolBlend {blendOption=Linear}
    boundBone {boneName=dyn_shorts_r_fr}
    componentLinearAccScale=<1 1 1>
    componentLinearVelScale=<1 1 1>
    gravityOverride {z=-980}
    physicsBodyDefinitions [{
      boundBone {boneName=dyn_shorts_r_fr}
      boxExtents=<10 10 10>
      constraintSetup {angularLimitsMax=<1 0 0>, angularLimitsMin=<-1 -20 0>, angularTarget=<1 0 0>}
      localJointOffset=<0.4 0 0>
     }]
    relativeSpaceBone {boneName=thigh_r}
   }
   n19 = Anim Dynamics - Bone: dyn_shorts_r_bk!(ComponentPose=@n18, Alpha=1, GravityOverride:Vector="0,0,-980", ExternalForce="0,0,0") -> (Pose) {
    angularDampingOverride=1
    angularSpringConstant=500
    bAngularSpring=true
    bOverrideAngularDamping=true
    bUseGravityOverride=true
    gravityScale=0.35
    linearSpringConstant=2000
    simulationSpace=RootRelative
    alphaBoolBlend {blendOption=Linear}
    boundBone {boneName=dyn_shorts_r_bk}
    componentAppliedLinearAccClamp=<50000 50000 50000>
    componentLinearAccScale=<1 1 1>
    componentLinearVelScale=<1 1 1>
    gravityOverride {z=-980}
    physicsBodyDefinitions [{
      boundBone {boneName=dyn_shorts_r_bk}
      boxExtents=<10 10 10>
      constraintSetup {angularLimitsMax=<0 4 0>, angularLimitsMin=<0 -20 0>, angularTarget=<1 0 0>}
      localJointOffset=<0.4 0 0>
     }]
    relativeSpaceBone {boneName=thigh_r}
   }
   Output Pose(Result=Component To Local(ComponentPose=@n19)) {layerGroup=None}
   n20 = "Constraint - beltcorr (PPPPPPP)"!(
    ComponentPose=Local To Component(LocalPose=Copy Pose From Mesh() {bUseAttachedParent=true})
    Alpha=1
    ConstraintWeights_0=0.2932
    ConstraintWeights_1:float=0
    ConstraintWeights_2:float=0
    ConstraintWeights_3:float=0.0034
    ConstraintWeights_4:float=0.4218
    ConstraintWeights_5:float=0.2794
    ConstraintWeights_6:float=0.0021
   ) -> (Pose) {
    boneToModify {boneName=beltcorr}
    constraintSetup [
     {targetBone {boneName=pelvis}}
     {targetBone {boneName=spine_01}}
     {targetBone {boneName=spine_02}}
     {targetBone {boneName=thigh_l}}
     {targetBone {boneName=thigh_twist_01_l}}
     {targetBone {boneName=deform_glute_l}}
     {targetBone {boneName=thigh_twist_01_r}}
    ]
    constraintWeights [1,1,1,1,1,1,1]
   }
   n21 = Control Rig!(
    Source=Component To Local(ComponentPose=@n20)
    Alpha
    MovFrvBckv:double=Get Forward Backward Movement() -> (Forward Backward Movement:double)
    MoveLeftRight:double=Get Left Right Movement() -> (Left Right Movement:double)
   ) -> (Pose) {controlRigAssetReference {blueprintRigClass=&/M_MED_SolarDusk_Spike_CtrlRig.M_MED_SolarDusk_Spike_CtrlRig_C}}
   Save cached pose 'frontnofront'(Pose=@n21) {cacheName=frontnofront}
   n22 = Anim Dynamics - Bone: Dyn_belt!(
    ComponentPose=Local To Component(LocalPose=Use cached pose 'frontnofront'())
    Alpha=1
    GravityOverride:Vector="0,0,-980"
    ExternalForce="0,0,0"
   ) -> (Pose) {
    angularDampingOverride=1
    angularSpringConstant=500
    bAngularSpring=true
    bOverrideAngularDamping=true
    bUseGravityOverride=true
    gravityScale=0.35
    linearSpringConstant=2000
    simulationSpace=RootRelative
    alphaBoolBlend {blendOption=Linear}
    boundBone {boneName=Dyn_belt}
    componentLinearAccScale=<1 1 1>
    componentLinearVelScale=<1 1 1>
    gravityOverride {z=-980}
    physicsBodyDefinitions [{
      boundBone {boneName=Dyn_belt}
      boxExtents=<10 10 10>
      constraintSetup {angularLimitsMax=<1 0 0>, angularLimitsMin=<-1 -10 0>, angularTarget=<1 0 0>}
      localJointOffset=<1 0 0>
     }]
    relativeSpaceBone {boneName=pelvis}
   }
   n23 = Anim Dynamics - Bone: dyn_Badge_01!(ComponentPose=@n22, Alpha=1, GravityOverride:Vector="0,0,-980", ExternalForce="0,0,0") -> (Pose) {
    angularDampingOverride=1
    angularSpringConstant=600
    bAngularSpring=true
    bOverrideAngularDamping=true
    bUseGravityOverride=true
    gravityScale=0.35
    linearSpringConstant=2000
    simulationSpace=RootRelative
    alphaBoolBlend {blendOption=Linear}
    boundBone {boneName=dyn_Badge_01}
    componentLinearAccScale=<1 1 1>
    componentLinearVelScale=<1 1 1>
    gravityOverride {z=-980}
    physicsBodyDefinitions [{
      boundBone {boneName=dyn_Badge_01}
      boxExtents=<10 10 10>
      constraintSetup {angularLimitsMax=<1 0 0>, angularLimitsMin=<-1 -20 -30>, angularTarget=<1 0 0>}
      localJointOffset=<0.1 0 0>
     }]
    relativeSpaceBone {boneName=pelvis}
    simSpaceSettings {maxAngularAcceleration=45, maxAngularVelocity=10, simSpaceAngularAlpha=1}
   }
   n24 = Anim Dynamics - Bone: dyn_Badge_02!(ComponentPose=@n23, Alpha=1, GravityOverride:Vector="0,0,-980", ExternalForce="0,0,0") -> (Pose) {
    angularDampingOverride=1
    angularSpringConstant=600
    bAngularSpring=true
    bOverrideAngularDamping=true
    bUseGravityOverride=true
    gravityScale=0.35
    linearSpringConstant=2000
    simulationSpace=RootRelative
    alphaBoolBlend {blendOption=Linear}
    boundBone {boneName=dyn_Badge_02}
    componentAppliedLinearAccClamp=<50000 50000 50000>
    componentLinearAccScale=<1 1 1>
    componentLinearVelScale=<1 1 1>
    gravityOverride {z=-980}
    physicsBodyDefinitions [{
      boundBone {boneName=dyn_Badge_02}
      boxExtents=<10 10 10>
      constraintSetup {angularLimitsMax=<0 4 0>, angularLimitsMin=<0 -20 0>, angularTarget=<1 0 0>}
      localJointOffset=<0.1 0 0>
     }]
    relativeSpaceBone {boneName=pelvis}
    simSpaceSettings {maxAngularAcceleration=45, maxAngularVelocity=10, simSpaceAngularAlpha=1}
   }
   n25 = Anim Dynamics - Bone: dyn_fur_01!(ComponentPose=@n24, Alpha=1, GravityOverride:Vector="0,0,-980", ExternalForce="0,0,0") -> (Pose) {
    angularDampingOverride=1
    angularSpringConstant=600
    bAngularSpring=true
    bOverrideAngularDamping=true
    bUseGravityOverride=true
    gravityScale=0.35
    linearSpringConstant=2000
    simulationSpace=RootRelative
    alphaBoolBlend {blendOption=Linear}
    boundBone {boneName=dyn_fur_01}
    componentLinearAccScale=<1 1 1>
    componentLinearVelScale=<1 1 1>
    gravityOverride {z=-980}
    physicsBodyDefinitions [{
      boundBone {boneName=dyn_fur_01}
      boxExtents=<10 10 10>
      constraintSetup {angularLimitsMax=<1 0 0>, angularLimitsMin=<-1 -50 -90>, angularTarget=<1 0 0>}
      localJointOffset=<0.1 0 0>
     }]
    relativeSpaceBone {boneName=pelvis}
    simSpaceSettings {maxAngularAcceleration=45, maxAngularVelocity=10, simSpaceAngularAlpha=1}
   }
   n26 = Anim Dynamics - Bone: dyn_fur_02!(ComponentPose=@n25, Alpha=1, GravityOverride:Vector="0,0,-980", ExternalForce="0,0,0") -> (Pose) {
    angularDampingOverride=1
    angularSpringConstant=600
    bAngularSpring=true
    bOverrideAngularDamping=true
    bUseGravityOverride=true
    gravityScale=0.35
    linearSpringConstant=2000
    simulationSpace=RootRelative
    alphaBoolBlend {blendOption=Linear}
    boundBone {boneName=dyn_fur_02}
    componentAppliedLinearAccClamp=<50000 50000 50000>
    componentLinearAccScale=<1 1 1>
    componentLinearVelScale=<1 1 1>
    gravityOverride {z=-980}
    physicsBodyDefinitions [{
      boundBone {boneName=dyn_fur_02}
      boxExtents=<10 10 10>
      constraintSetup {angularLimitsMax=<0 4 0>, angularLimitsMin=<0 -20 -60>, angularTarget=<1 0 0>}
      localJointOffset=<0.1 0 0>
     }]
    relativeSpaceBone {boneName=pelvis}
    simSpaceSettings {maxAngularAcceleration=45, maxAngularVelocity=10, simSpaceAngularAlpha=1}
   }
   Save cached pose 'fe'(Pose=Component To Local(ComponentPose=@n26)) {cacheName=fe}
  ]}
 EventGraph {nodes [
   Event BlueprintUpdateAnimation()
   > n1 = PlayerMovement(AnimInstance=Self-Reference(), XDotProductCoefficient=1, YDotProductCoefficient=1, ZDotProductCoefficient=1)
   > Set Left Right Movement(Left Right Movement:double=@n1) -> (Output_Get:double)
   > Set Forward Backward Movement(Forward Backward Movement:double=@n1.YMovement) -> (Output_Get:double)
  ]}
]
variables [
 Forward Backward Movement {type=number}
 Left Right Movement {type=number}
]
