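# Physics asset for the Spike (SolarDusk) male character's FaceAcc backpack part, defining physics for hair bones over an upper-body skeleton.
# Box bodies form center, left and right hair chains from small base capsules, with narrow angular limits and side branches at segment 02, the right one locked.
# Capsules cover head, neck, clavicles and upper spine, joined by locked constraints.
# Angularly free constraints also link bases to head, side tips to center, and center to spine_05.
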PhysicsAsset /BRCosmetics/Characters/Player/Male/Medium/Bodies/M_MED_SolarDusk_Spike/Meshes/Parts/M_MED_SolarDusk_Spike_FaceAcc_backpack
bodies {
 Hair_01_c [{shapeType=Box, extentX=16.3011, extentY=9.2483, extentZ=3.5409, center=<1 0 0>}]
 Hair_01_ls [{shapeType=Box, extentX=16.3011, extentY=9.2483, extentZ=3.5409, center=<1 0 0>}]
 Hair_01_rs [{shapeType=Box, extentX=16.3011, extentY=9.2483, extentZ=3.5409, center=<1 0 0>}]
 Hair_02_c [{shapeType=Box, extentX=13.9786, extentY=9.2483, extentZ=3.5409, center=<2.7063 0 0>}]
 Hair_02_ls [{shapeType=Box, extentX=10.5444, extentY=9.4123, extentZ=3.5409, center=<1.2784 0 0>}]
 Hair_02_rs [{shapeType=Box, extentX=10.5444, extentY=9.4123, extentZ=3.5409, center=<1.2784 0 0>}]
 Hair_03_c [{shapeType=Box, extentX=9.4203, extentY=9.2483, extentZ=4.2709, center=<4 0 0>, rotation=<rot 3.075 0 0>}]
 Hair_03_ls [{shapeType=Box, extentX=8.4293, extentY=8.0675, extentZ=3.247, center=<3.0589 0 0>, rotation=<rot 3.4616 0 0>}]
 Hair_03_rs [{shapeType=Box, extentX=8.4293, extentY=8.0675, extentZ=3.247, center=<3.0589 0 0>, rotation=<rot 3.4616 0 0>}]
 Hair_04_ls [{shapeType=Box, extentX=3.4919, extentY=2.3174, extentZ=2.3174, center=<1.1047 0 0>}]
 Hair_04_rs [{shapeType=Box, extentX=4.8394, extentY=2.554, extentZ=2.3875, center=<2.0242 0 0>}]
 Hair_Bsae_C [{shapeType=Capsule, length=0.505, radius=0.505, shapeName=Hair_Bsae_C_capsule, rotation=<rot 0 -90 0>}]
 Hair_Bsae_L [{shapeType=Capsule, length=0.505, radius=0.505, shapeName=Hair_Bsae_L_capsule, rotation=<rot 0 -90 0>}]
 Hair_Bsae_R [{shapeType=Capsule, length=0.505, radius=0.505, shapeName=Hair_Bsae_R_capsule, rotation=<rot 0 -90 0>}]
 clavicle_l [{shapeType=Capsule, length=7.2419, radius=6.4543, shapeName=clavicle_l_capsule, center=<10.714 0 0>, rotation=<rot 90 180 0>}]
 clavicle_r [{shapeType=Capsule, length=7.2419, radius=6.4543, shapeName=clavicle_r_capsule, center=<-10.714 0 0>, rotation=<rot 90 0 0>}]
 head [{shapeType=Capsule, length=4.5629, radius=5.4683, center=<0 4.6185 0>, rotation=<rot 90 0 0>}]
 neck_01 [{shapeType=Capsule, length=3.8257, radius=4.6006, rotation=<rot -90 -0.0001 0>}]
 neck_02 [{shapeType=Capsule, length=1.8107, radius=4.0141, rotation=<rot 90 0 0>}]
 spine_04 [{shapeType=Capsule, length=8.6735, radius=11.5393, shapeName=spine_04_capsule, center=<4.1408 0 0>, rotation=<rot 0 -90 0>}]
 spine_05 [{shapeType=Capsule, length=17.536, radius=11.0084, center=<0.5066 1.4021 0>}]
}
constraints [
 {
  bone1Name=head
  bone2Name=neck_02
  swing1LimitDegrees=45
  swing1Motion=Locked
  swing2LimitDegrees=45
  swing2Motion=Locked
  twistLimitDegrees=45
  twistMotion=Locked
 }
 {
  bone1Name=neck_02
  bone2Name=neck_01
  swing1LimitDegrees=45
  swing1Motion=Locked
  swing2LimitDegrees=45
  swing2Motion=Locked
  twistLimitDegrees=45
  twistMotion=Locked
 }
 {
  bone1Name=neck_01
  bone2Name=spine_05
  swing1LimitDegrees=45
  swing1Motion=Locked
  swing2LimitDegrees=45
  swing2Motion=Locked
  twistLimitDegrees=45
  twistMotion=Locked
 }
 {
  bone1Name=clavicle_r
  bone2Name=spine_05
  swing1LimitDegrees=45
  swing1Motion=Locked
  swing2LimitDegrees=45
  swing2Motion=Locked
  twistLimitDegrees=45
  twistMotion=Locked
 }
 {
  bone1Name=clavicle_l
  bone2Name=spine_05
  swing1LimitDegrees=45
  swing1Motion=Locked
  swing2LimitDegrees=45
  swing2Motion=Locked
  twistLimitDegrees=45
  twistMotion=Locked
 }
 {
  bone1Name=spine_05
  bone2Name=spine_04
  swing1LimitDegrees=45
  swing1Motion=Locked
  swing2LimitDegrees=45
  swing2Motion=Locked
  twistLimitDegrees=45
  twistMotion=Locked
 }
 {
  bone1Name=Hair_04_rs
  bone2Name=Hair_02_rs
  swing1LimitDegrees=10
  swing1Motion=Locked
  swing2LimitDegrees=10
  swing2Motion=Locked
  twistLimitDegrees=1
  twistMotion=Locked
 }
 {bone1Name=Hair_01_c, bone2Name=Hair_Bsae_C, swing1LimitDegrees=10, swing2LimitDegrees=25, twistLimitDegrees=3}
 {bone1Name=Hair_02_c, bone2Name=Hair_01_c, swing1LimitDegrees=10, swing2LimitDegrees=40, twistLimitDegrees=3}
 {bone1Name=Hair_03_c, bone2Name=Hair_02_c, swing1LimitDegrees=10, swing2LimitDegrees=20, twistLimitDegrees=3, twistMotion=Locked}
 {bone1Name=Hair_01_ls, bone2Name=Hair_Bsae_L, swing1LimitDegrees=10, swing2LimitDegrees=25, twistLimitDegrees=3}
 {bone1Name=Hair_02_ls, bone2Name=Hair_01_ls, swing1LimitDegrees=10, swing2LimitDegrees=25, twistLimitDegrees=3}
 {bone1Name=Hair_03_ls, bone2Name=Hair_02_ls, swing1LimitDegrees=10, swing2LimitDegrees=25, twistLimitDegrees=3}
 {bone1Name=Hair_04_ls, bone2Name=Hair_02_ls, swing1LimitDegrees=20, swing2LimitDegrees=20, twistLimitDegrees=1}
 {bone1Name=Hair_01_rs, bone2Name=Hair_Bsae_R, swing1LimitDegrees=10, swing2LimitDegrees=25, twistLimitDegrees=3}
 {bone1Name=Hair_02_rs, bone2Name=Hair_01_rs, swing1LimitDegrees=10, swing2LimitDegrees=25, twistLimitDegrees=3}
 {bone1Name=Hair_03_rs, bone2Name=Hair_02_rs, swing1LimitDegrees=10, swing2LimitDegrees=25, twistLimitDegrees=3}
 {
  bone1Name=Hair_03_rs
  bone2Name=Hair_03_c
  swing1LimitDegrees=45
  swing1Motion=Free
  swing2LimitDegrees=45
  swing2Motion=Free
  twistLimitDegrees=45
  twistMotion=Free
 }
 {
  bone1Name=Hair_03_ls
  bone2Name=Hair_03_c
  swing1LimitDegrees=45
  swing1Motion=Free
  swing2LimitDegrees=45
  swing2Motion=Free
  twistLimitDegrees=45
  twistMotion=Free
 }
 {
  bone1Name=Hair_03_c
  bone2Name=spine_05
  swing1LimitDegrees=45
  swing1Motion=Free
  swing2LimitDegrees=45
  swing2Motion=Free
  twistLimitDegrees=45
  twistMotion=Free
 }
 {
  bone1Name=Hair_02_c
  bone2Name=spine_05
  swing1LimitDegrees=45
  swing1Motion=Free
  swing2LimitDegrees=45
  swing2Motion=Free
  twistLimitDegrees=45
  twistMotion=Free
 }
 {
  bone1Name=Hair_03_c
  bone2Name=spine_05
  swing1LimitDegrees=45
  swing1Motion=Free
  swing2LimitDegrees=45
  swing2Motion=Free
  twistLimitDegrees=45
  twistMotion=Free
 }
 {
  bone1Name=Hair_Bsae_C
  bone2Name=head
  swing1LimitDegrees=45
  swing1Motion=Free
  swing2LimitDegrees=45
  swing2Motion=Free
  twistLimitDegrees=45
  twistMotion=Free
 }
 {
  bone1Name=Hair_Bsae_L
  bone2Name=head
  swing1LimitDegrees=45
  swing1Motion=Free
  swing2LimitDegrees=45
  swing2Motion=Free
  twistLimitDegrees=45
  twistMotion=Free
 }
 {
  bone1Name=Hair_Bsae_R
  bone2Name=head
  swing1LimitDegrees=45
  swing1Motion=Free
  swing2LimitDegrees=45
  swing2Motion=Free
  twistLimitDegrees=45
  twistMotion=Free
 }
]
